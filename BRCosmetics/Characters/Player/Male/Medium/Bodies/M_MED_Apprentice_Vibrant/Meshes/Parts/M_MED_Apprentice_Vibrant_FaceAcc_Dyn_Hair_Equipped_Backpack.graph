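# Physics asset for the dynamic hair of the Apprentice Vibrant male character, apparently the variant used with a backpack equipped.
# Capsules on spine, neck, head and oversized clavicles form the body, linked by free-rotation constraints.
# Four hair chains (back, back-left, left, right) run from tiny master anchors through two tightly limited segments.
# The left anchor joins the neck, the others the head, and three braid bodies each constrain directly to the head rather than chaining.
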PhysicsAsset /BRCosmetics/Characters/Player/Male/Medium/Bodies/M_MED_Apprentice_Vibrant/Meshes/Parts/M_MED_Apprentice_Vibrant_FaceAcc_Dyn_Hair_Equipped_Backpack
bodies {
 clavicle_l [{
   shapeType=Capsule
   length=4.125
   radius=12.5
   shapeName=clavicle_l_capsule
   center=<9.5853 1.8908 -3.1437>
   rotation=<rot 81.2706 2.8662 -2.1481>
  }]
 clavicle_r [{
   shapeType=Capsule
   length=4.125
   radius=12.5
   shapeName=clavicle_l_capsule
   center=<-9.5853 -1.8908 3.1437>
   rotation=<rot 81.2706 2.8662 177.8519>
  }]
 dyn_braids_01 [{shapeType=Capsule, length=3.75, radius=6.25, shapeName=dyn_hair_bk_01_capsule, center=<2 0 0>, rotation=<rot 90 0 0>}]
 dyn_braids_02 [{shapeType=Capsule, length=3.75, radius=6.25, shapeName=dyn_hair_bk_01_capsule, center=<5 0 0>, rotation=<rot 90 0 0>}]
 dyn_braids_03 [{shapeType=Capsule, length=3.75, radius=6.25, shapeName=dyn_hair_bk_01_capsule, center=<2 0 0>, rotation=<rot 90 0 0>}]
 dyn_hair_bk_01 [{shapeType=Capsule, length=3.75, radius=6.25, shapeName=dyn_hair_bk_01_capsule, center=<5 0 5>, rotation=<rot 90 0 0>}]
 dyn_hair_bk_02 [{shapeType=Capsule, length=5.25, radius=13.25, shapeName=dyn_hair_bk_01_capsule, center=<12 0 6>, rotation=<rot 90 0 0>}]
 dyn_hair_bk_MSTR [{shapeType=Capsule, length=0.505, radius=0.505, shapeName=dyn_hair_bk_MSTR_capsule, rotation=<rot 0 -90 0>}]
 dyn_hair_bk_l_01 [{shapeType=Capsule, length=4, radius=6, shapeName=dyn_hair_bk_01_capsule, center=<5 0 5>, rotation=<rot 84.9999 -0.0001 -0.0002>}]
 dyn_hair_bk_l_02 [{shapeType=Capsule, length=7.75, radius=13.25, shapeName=dyn_hair_bk_01_capsule, center=<10 -3 6>, rotation=<rot 1.7079 29.7002 70.0699>}]
 dyn_hair_bk_l_MSTR [{shapeType=Capsule, length=0.505, radius=0.505, shapeName=dyn_hair_bk_MSTR_capsule, rotation=<rot 0 -90 0>}]
 dyn_hair_l_01 [{shapeType=Capsule, length=5.25, radius=4.75, shapeName=dyn_hair_bk_01_capsule, center=<5 0 5>, rotation=<rot 0 0 90>}]
 dyn_hair_l_02 [{shapeType=Capsule, length=7, radius=5, shapeName=dyn_hair_bk_01_capsule, center=<4 -2.5 5.5>, rotation=<rot 60 70.0003 180>}]
 dyn_hair_l_MSTR [{shapeType=Capsule, length=0.505, radius=0.505, shapeName=dyn_hair_l_MSTR_capsule, rotation=<rot 0 -90 0>}]
 dyn_hair_r_01 [{shapeType=Capsule, length=3.75, radius=6.25, shapeName=dyn_hair_r_01_capsule, center=<5 0 5>, rotation=<rot 90 0 0>}]
 dyn_hair_r_02 [{shapeType=Capsule, length=5.75, radius=9, shapeName=dyn_hair_r_01_capsule, center=<8 0 2>, rotation=<rot 0 -20 100>}]
 dyn_hair_r_MSTR [{shapeType=Capsule, length=0.505, radius=0.505, shapeName=dyn_hair_r_MSTR_capsule, rotation=<rot 0 -90 0>}]
 head [{shapeType=Capsule, length=5, radius=11.125, shapeName=spine_05_capsule, center=<10.1906 -1.5652 0.6368>, rotation=<rot 90 0 -14.9999>}]
 neck_01 [{shapeType=Capsule, length=4.75, radius=11.125, shapeName=spine_05_capsule, center=<3.5798 0.6047 0>, rotation=<rot 90 0 10.0002>}]
 spine_05 [{shapeType=Capsule, length=6.25, radius=15.625, shapeName=spine_05_capsule, center=<-3.1677 0.4203 0>}]
}
constraints [
 {
  bone1Name=dyn_hair_l_MSTR
  bone2Name=neck_01
  swing1LimitDegrees=45
  swing1Motion=Free
  swing2LimitDegrees=45
  swing2Motion=Free
  twistLimitDegrees=45
  twistMotion=Free
 }
 {bone1Name=dyn_hair_l_01, bone2Name=dyn_hair_l_MSTR, swing1LimitDegrees=15, swing2LimitDegrees=20, twistLimitDegrees=2}
 {
  bone1Name=dyn_hair_bk_MSTR
  bone2Name=head
  swing1LimitDegrees=45
  swing1Motion=Free
  swing2LimitDegrees=45
  swing2Motion=Free
  twistLimitDegrees=45
  twistMotion=Free
 }
 {bone1Name=dyn_hair_bk_01, bone2Name=dyn_hair_bk_MSTR, swing1LimitDegrees=15, swing2LimitDegrees=10, twistLimitDegrees=2}
 {
  bone1Name=dyn_hair_r_MSTR
  bone2Name=head
  swing1LimitDegrees=45
  swing1Motion=Free
  swing2LimitDegrees=45
  swing2Motion=Free
  twistLimitDegrees=45
  twistMotion=Free
 }
 {bone1Name=dyn_hair_r_01, bone2Name=dyn_hair_r_MSTR, swing1LimitDegrees=15, swing2LimitDegrees=10, twistLimitDegrees=2}
 {
  bone1Name=dyn_hair_bk_l_MSTR
  bone2Name=head
  swing1LimitDegrees=45
  swing1Motion=Free
  swing2LimitDegrees=45
  swing2Motion=Free
  twistLimitDegrees=45
  twistMotion=Free
 }
 {bone1Name=dyn_hair_bk_l_01, bone2Name=dyn_hair_bk_l_MSTR, swing1LimitDegrees=15, swing2LimitDegrees=10, twistLimitDegrees=2}
 {
  bone1Name=clavicle_r
  bone2Name=spine_05
  swing1LimitDegrees=45
  swing1Motion=Free
  swing2LimitDegrees=45
  swing2Motion=Free
  twistLimitDegrees=45
  twistMotion=Free
 }
 {
  bone1Name=head
  bone2Name=neck_01
  swing1LimitDegrees=45
  swing1Motion=Free
  swing2LimitDegrees=45
  swing2Motion=Free
  twistLimitDegrees=45
  twistMotion=Free
 }
 {bone1Name=dyn_hair_bk_l_02, bone2Name=dyn_hair_bk_l_01, swing1LimitDegrees=15, swing2LimitDegrees=10, twistLimitDegrees=2}
 {bone1Name=dyn_hair_r_02, bone2Name=dyn_hair_r_01, swing1LimitDegrees=15, swing2LimitDegrees=10, twistLimitDegrees=2}
 {bone1Name=dyn_hair_bk_02, bone2Name=dyn_hair_bk_01, swing1LimitDegrees=15, swing2LimitDegrees=10, twistLimitDegrees=2}
 {
  bone1Name=neck_01
  bone2Name=spine_05
  swing1LimitDegrees=45
  swing1Motion=Free
  swing2LimitDegrees=45
  swing2Motion=Free
  twistLimitDegrees=45
  twistMotion=Free
 }
 {bone1Name=dyn_hair_l_02, bone2Name=dyn_hair_l_01, swing1LimitDegrees=30, swing2LimitDegrees=20, twistLimitDegrees=2}
 {bone1Name=dyn_braids_01, bone2Name=head, swing1LimitDegrees=30, swing2LimitDegrees=30, twistLimitDegrees=10}
 {bone1Name=dyn_braids_02, bone2Name=head, swing1LimitDegrees=30, swing2LimitDegrees=30, twistLimitDegrees=10}
 {bone1Name=dyn_braids_03, bone2Name=head, swing1LimitDegrees=30, swing2LimitDegrees=30, twistLimitDegrees=10}
 {
  bone1Name=clavicle_l
  bone2Name=spine_05
  swing1LimitDegrees=45
  swing1Motion=Free
  swing2LimitDegrees=45
  swing2Motion=Free
  twistLimitDegrees=45
  twistMotion=Free
 }
]
