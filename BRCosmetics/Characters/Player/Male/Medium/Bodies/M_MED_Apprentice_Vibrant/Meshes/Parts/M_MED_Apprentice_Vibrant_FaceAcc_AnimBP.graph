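# Post-process style hair animation for the Apprentice Vibrant face accessory, layering secondary motion onto copied body poses.
# Look-At bones prepare the hair, then head-rotation alphas drive corrective offsets before RigidBody simulation, or AnimDynamics when backbling is worn or rigid bodies are disabled.
# An elimination event switches to reactive physics plus additive animation for six seconds.
# The material-driven hair-tip dynamics chain is not connected to the output.
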AnimBlueprint /BRCosmetics/Characters/Player/Male/Medium/Bodies/M_MED_Apprentice_Vibrant/Meshes/Parts/M_MED_Apprentice_Vibrant_FaceAcc_AnimBP
target_skeleton=&/M_MED_Apprentice_Vibrant_FaceAcc_Skeleton
graphs [
 AnimGraph {nodes [
   n1 = Get ReactiveAlpha() -> (ReactiveAlpha:double)
   n2 = Anim Dynamics - Bone: dyn_rect_hairTips_A_01!(ComponentPose, Alpha=@n1, GravityScale:float=1, ExternalForce="0,0,0") -> (Pose) {
    angularDampingOverride=0.999
    angularSpringConstant=500
    bAngularSpring=true
    bGravityOverrideInSimSpace=true
    bOverrideAngularDamping=true
    bUseGravityOverride=true
    simulationSpace=RootRelative
    alphaBoolBlend {blendOption=Linear}
    boundBone {boneName=dyn_rect_hairTips_A_01}
    componentAppliedLinearAccClamp=<1000 1000 1000>
    componentLinearAccScale=<0.2 0.2 0.2>
    componentLinearVelScale=<0.6 0.6 0.6>
    gravityOverride {z=-600}
    physicsBodyDefinitions [{
      sphereCollisionRadius=0
      boundBone {boneName=dyn_rect_hairTips_A_01}
      boxExtents=<10 15 15>
      constraintSetup {angularLimitsMax=<0 0 10>, angularLimitsMin=<0 0 -15>, angularTarget=<0 0 0>}
      localJointOffset=<3 -1 0>
     }]
    relativeSpaceBone {boneName=head}
   }
   n3 = Anim Dynamics - Bone: dyn_rect_hairTips_B_01!(ComponentPose=@n2, Alpha=@n1, GravityScale:float=1, ExternalForce="0,0,0") -> (Pose) {
    angularDampingOverride=0.999
    angularSpringConstant=500
    bAngularSpring=true
    bGravityOverrideInSimSpace=true
    bOverrideAngularDamping=true
    bUseGravityOverride=true
    simulationSpace=RootRelative
    alphaBoolBlend {blendOption=Linear}
    boundBone {boneName=dyn_rect_hairTips_B_01}
    componentAppliedLinearAccClamp=<1000 1000 1000>
    componentLinearAccScale=<0.2 0.2 0.2>
    componentLinearVelScale=<0.6 0.6 0.6>
    gravityOverride {z=-600}
    physicsBodyDefinitions [{
      sphereCollisionRadius=0
      boundBone {boneName=dyn_rect_hairTips_B_01}
      boxExtents=<10 16 16>
      constraintSetup {angularLimitsMax=<0 0 12>, angularLimitsMin=<0 0 -15>, angularTarget=<0 0 0>}
      localJointOffset=<4 -2.2 0>
     }]
    relativeSpaceBone {boneName=head}
   }
   n4 = Anim Dynamics - Bone: dyn_rect_hairTips_C_01!(ComponentPose=@n3, Alpha=@n1, GravityScale:float=1, ExternalForce="0,0,0") -> (Pose) {
    angularDampingOverride=0.999
    angularSpringConstant=400
    bAngularSpring=true
    bGravityOverrideInSimSpace=true
    bOverrideAngularDamping=true
    bUseGravityOverride=true
    simulationSpace=RootRelative
    alphaBoolBlend {blendOption=Linear}
    boundBone {boneName=dyn_rect_hairTips_C_01}
    componentAppliedLinearAccClamp=<1000 1000 1000>
    componentLinearAccScale=<0.4 0.4 0.4>
    componentLinearVelScale=<0.6 0.6 0.6>
    gravityOverride {z=-980}
    physicsBodyDefinitions [{
      sphereCollisionRadius=0
      boundBone {boneName=dyn_rect_hairTips_C_01}
      boxExtents=<15 25 20>
      constraintSetup {angularLimitsMax=<0 10 18>, angularLimitsMin=<0 -10 -12>, angularTarget=<0 0 0>}
      localJointOffset=<3 -2 -1>
     }]
    relativeSpaceBone {boneName=head}
   }
   n5 = Anim Dynamics - Bone: dyn_rect_hairTips_D_01!(ComponentPose=@n4, Alpha=@n1, GravityScale:float=1, ExternalForce="0,0,0") -> (Pose) {
    angularDampingOverride=0.999
    angularSpringConstant=500
    bAngularSpring=true
    bGravityOverrideInSimSpace=true
    bOverrideAngularDamping=true
    bUseGravityOverride=true
    simulationSpace=RootRelative
    alphaBoolBlend {blendOption=Linear}
    boundBone {boneName=dyn_rect_hairTips_D_01}
    componentAppliedLinearAccClamp=<1000 1000 1000>
    componentLinearAccScale=<0.3 0.3 0.3>
    componentLinearVelScale=<0.6 0.6 0.6>
    gravityOverride {z=-980}
    physicsBodyDefinitions [{
      sphereCollisionRadius=0
      boundBone {boneName=dyn_rect_hairTips_D_01}
      boxExtents=<10 15 15>
      constraintSetup {angularLimitsMax=<0 10 10>, angularLimitsMin=<0 -20 -10>, angularTarget=<0 0 0>}
      localJointOffset=<1 -0.5 -2>
     }]
    relativeSpaceBone {boneName=head}
   }
   n6 = Anim Dynamics - Bone: dyn_rect_hairTips_E_01!(ComponentPose=@n5, Alpha=@n1, GravityScale:float=1, ExternalForce="0,0,0") -> (Pose) {
    angularDampingOverride=0.999
    angularSpringConstant=500
    bAngularSpring=true
    bGravityOverrideInSimSpace=true
    bOverrideAngularDamping=true
    bUseGravityOverride=true
    simulationSpace=RootRelative
    alphaBoolBlend {blendOption=Linear}
    boundBone {boneName=dyn_rect_hairTips_E_01}
    componentAppliedLinearAccClamp=<1000 1000 1000>
    componentLinearAccScale=<0.3 0.3 0.3>
    componentLinearVelScale=<0.6 0.6 0.6>
    gravityOverride {z=-980}
    physicsBodyDefinitions [{
      sphereCollisionRadius=0
      boundBone {boneName=dyn_rect_hairTips_E_01}
      boxExtents=<10 15 15>
      constraintSetup {angularLimitsMax=<0 20 10>, angularLimitsMin=<0 -20 -30>, angularTarget=<0 0 0>}
      localJointOffset=<2 0 -2>
     }]
    relativeSpaceBone {boneName=head}
   }
   n7 = Anim Dynamics - Bone: dyn_rect_hairTips_F_01!(ComponentPose=@n6, Alpha=@n1, GravityScale:float=1, ExternalForce="0,0,0") -> (Pose) {
    angularDampingOverride=0.999
    angularSpringConstant=1000
    bAngularSpring=true
    bGravityOverrideInSimSpace=true
    bOverrideAngularDamping=true
    bUseGravityOverride=true
    simulationSpace=RootRelative
    alphaBoolBlend {blendOption=Linear}
    boundBone {boneName=dyn_rect_hairTips_F_01}
    componentAppliedLinearAccClamp=<1000 1000 1000>
    componentLinearAccScale=<0.3 0.3 0.3>
    componentLinearVelScale=<0.6 0.6 0.6>
    gravityOverride {z=-980}
    physicsBodyDefinitions [{
      sphereCollisionRadius=0
      boundBone {boneName=dyn_rect_hairTips_F_01}
      boxExtents=<10 15 15>
      constraintSetup {angularLimitsMax=<0 20 15>, angularLimitsMin=<0 -10 -15>, angularTarget=<0 0 0>}
      localJointOffset=<0 0 -3>
     }]
    relativeSpaceBone {boneName=head}
   }
   Anim Dynamics - Bone: dyn_rect_hairTips_G_01!(ComponentPose=@n7, Alpha=@n1, GravityScale:float=1, ExternalForce="0,0,0") -> (Pose) {
    angularDampingOverride=0.999
    angularSpringConstant=500
    bAngularSpring=true
    bGravityOverrideInSimSpace=true
    bOverrideAngularDamping=true
    bUseGravityOverride=true
    simulationSpace=RootRelative
    alphaBoolBlend {blendOption=Linear}
    boundBone {boneName=dyn_rect_hairTips_G_01}
    componentAppliedLinearAccClamp=<1000 1000 1000>
    componentLinearAccScale=<0.4 0.4 0.4>
    componentLinearVelScale=<0.6 0.6 0.6>
    gravityOverride {z=-980}
    physicsBodyDefinitions [{
      sphereCollisionRadius=0
      boundBone {boneName=dyn_rect_hairTips_G_01}
      boxExtents=<15 25 20>
      constraintSetup {angularLimitsMax=<0 15 30>, angularLimitsMin=<0 -15 -30>, angularTarget=<0 0 0>}
      localJointOffset=<3 0 -4>
     }]
    relativeSpaceBone {boneName=head}
   }
   n8 = Layered blend per bone!(
    BasePose=Use cached pose 'MeshPose'()
    BlendPoses_0="Sequence Player (additive) 'M_Med_ApprenticeVibrant_FaceAcc_Hair_Anim'"() {sequence=/BRCosmetics/Characters/Player/Male/Medium/Bodies/M_MED_Apprentice_Vibrant/Animation/M_Med_ApprenticeVibrant_FaceAcc_Hair_Anim}
    BlendPoses_1:Pose Link Structure
    BlendWeights_0
    BlendWeights_1:float=1
   ) -> (Pose) {
    blendPoses [{linkId=-1, sourceLinkId=-1}, {linkId=-1, sourceLinkId=-1}]
    blendWeights [1,1]
    layerSetup [{branchFilters [{boneName=head}]}, {branchFilters []}]
   }
   Blend Poses by bool(BlendPose_0=@n8, BlendPose_1=Use cached pose 'MeshPose'(), BlendTime_0=0, BlendTime_1=0.25, bActiveValue=Get onKill() -> (onKill:bool)) {blendType=Linear}
   n9 = RigidBody!(
    ComponentPose=Local To Component(LocalPose=Use cached pose 'prePoseBaseHair'())
    Alpha
    OverrideWorldGravity:Vector="0,0,-980"
    ExternalForce
    ComponentLinearAccScale:Vector="0.3,0.4,0.4"
   ) -> (Pose) {
    bClampLinearTranslationLimitToRefPose=true
    bOverrideWorldGravity=true
    overridePhysicsAsset=&/M_MED_Apprentice_Vibrant_FaceAcc_Dyn_Hair_Reactive
    simulationSpace=BaseBoneSpace
    alphaBoolBlend {blendOption=Linear}
    alphaScaleBiasClamp {interpSpeedDecreasing=4, interpSpeedIncreasing=6}
    baseBoneRef {boneName=spine_05}
    componentAppliedLinearAccClamp=<3000 3000 3000>
    componentLinearAccScale=<0.6 0.6 0.2>
    componentLinearVelScale=<0.35 0.35 0.35>
    overrideWorldGravity {z=-875}
    simSpaceSettings {maxAngularAcceleration=200, maxAngularVelocity=5, maxLinearAcceleration=500, maxLinearVelocity=200}
   }
   n10 = Component To Local(ComponentPose=@n9)
   n11 = "Sequence Player (additive) 'M_Med_ApprenticeVibrant_FaceAcc_Hair_Anim'"() {sequence=&M_Med_ApprenticeVibrant_FaceAcc_Hair_Anim}
   n12 = "Transform (Modify) Bone - Bone: dyn_hair_r_01"(
    ComponentPose=Local To Component(LocalPose=Use cached pose 'prePoseBaseHair'())
    Alpha=Get head_rot_r() -> (head_rot_r:double)
    Translation="0,2,0.007"
    Rotation="0,0,0"
    Scale="1,1,1"
   ) {
    rotationMode=BMM_Additive
    rotationSpace=BCS_BoneSpace
    translationMode=BMM_Additive
    translationSpace=BCS_BoneSpace
    alphaBoolBlend {blendOption=Linear}
    boneToModify {boneName=dyn_hair_r_01}
    translation {y=2, z=0.007}
   }
   n13 = "Constraint - dyn_hair_l_02 (P)"(ComponentPose=@n12, Alpha=1, ConstraintWeights_0=1) {
    alphaBoolBlend {blendOption=Linear}
    boneToModify {boneName=dyn_hair_l_02}
    constraintSetup [{targetBone {boneName=spine_05}}]
    constraintWeights [1]
   }
   n14 = "Constraint - dyn_hair_bk_02 (P)"(ComponentPose=@n13, Alpha=1, ConstraintWeights_0=1) {
    alphaBoolBlend {blendOption=Linear}
    boneToModify {boneName=dyn_hair_bk_02}
    constraintSetup [{targetBone {boneName=spine_05}}]
    constraintWeights [1]
   }
   n15 = "Constraint - dyn_hair_r_02 (PP)"!(ComponentPose=@n14, Alpha=1, ConstraintWeights_0=0.5, ConstraintWeights_1:float=0.5) -> (Pose) {
    alphaBoolBlend {blendOption=Linear}
    boneToModify {boneName=dyn_hair_r_02}
    constraintSetup [{targetBone {boneName=spine_05}}, {targetBone {boneName=clavicle_r}}]
    constraintWeights [1,1]
   }
   n16 = "Constraint - dyn_hair_bk_l_02 (P)"(ComponentPose=@n15, Alpha=1, ConstraintWeights_0=1) {
    alphaBoolBlend {blendOption=Linear}
    boneToModify {boneName=dyn_hair_bk_l_02}
    constraintSetup [{targetBone {boneName=spine_05}}]
    constraintWeights [1]
   }
   n17 = Get head_rot_bk() -> (head_rot_bk:double)
   n18 = "Transform (Modify) Bone - Bone: dyn_hair_bk_02"(ComponentPose=@n16, Alpha=@n17, Translation="12,0,0", Rotation="0,0,0", Scale="1,1,1") {
    rotationMode=BMM_Additive
    rotationSpace=BCS_BoneSpace
    translationMode=BMM_Additive
    translationSpace=BCS_BoneSpace
    alphaBoolBlend {blendOption=Linear}
    boneToModify {boneName=dyn_hair_bk_02}
    translation {x=12}
   }
   n19 = "Transform (Modify) Bone - Bone: dyn_hair_r_02"(ComponentPose=@n18, Alpha=@n17, Translation="12,7,3", Rotation="0,0,0", Scale="1,1,1") {
    rotationMode=BMM_Additive
    rotationSpace=BCS_BoneSpace
    translationMode=BMM_Additive
    translationSpace=BCS_BoneSpace
    alphaBoolBlend {blendOption=Linear}
    boneToModify {boneName=dyn_hair_r_02}
    translation=<12 7 3>
   }
   n20 = "Transform (Modify) Bone - Bone: dyn_hair_bk_l_02"(ComponentPose=@n19, Alpha=@n17, Translation="12,0,0", Rotation="0,0,0", Scale="1,1,1") {
    rotationMode=BMM_Additive
    rotationSpace=BCS_BoneSpace
    translationMode=BMM_Additive
    translationSpace=BCS_BoneSpace
    alphaBoolBlend {blendOption=Linear}
    boneToModify {boneName=dyn_hair_bk_l_02}
    translation {x=12}
   }
   n21 = "Transform (Modify) Bone - Bone: dyn_hair_l_02"(ComponentPose=@n20, Alpha=@n17, Translation="0,-9,0", Rotation="20,45,0", Scale="1,1,1") {
    rotationMode=BMM_Additive
    rotationSpace=BCS_BoneSpace
    translationMode=BMM_Additive
    translationSpace=BCS_BoneSpace
    alphaBoolBlend {blendOption=Linear}
    boneToModify {boneName=dyn_hair_l_02}
    rotation {pitch=20, yaw=45}
    translation {y=-9}
   }
   n22 = "Transform (Modify) Bone - Bone: dyn_hair_l_02"(ComponentPose=@n21, Alpha=Get head_twist_l() -> (head_twist_l:double), Translation="-5,-2,1", Rotation="0,0,0", Scale="1,1,1") {
    rotationMode=BMM_Additive
    rotationSpace=BCS_BoneSpace
    translationMode=BMM_Additive
    translationSpace=BCS_BoneSpace
    alphaBoolBlend {blendOption=Linear}
    boneToModify {boneName=dyn_hair_l_02}
    translation=<-5 -2 1>
   }
   n23 = Get head_rot_r()
   n24 = "Transform (Modify) Bone - Bone: dyn_hair_bk_l_02"(ComponentPose=@n22, Alpha=@n23, Translation="5,-4.494,0.396", Rotation="10,0,0", Scale="1,1,1") {
    rotationMode=BMM_Additive
    rotationSpace=BCS_BoneSpace
    translationMode=BMM_Additive
    translationSpace=BCS_BoneSpace
    alphaBoolBlend {blendOption=Linear}
    boneToModify {boneName=dyn_hair_bk_l_02}
    rotation {pitch=10}
    translation=<5 -4.494 0.396>
   }
   n25 = "Transform (Modify) Bone - Bone: dyn_hair_bk_02"(ComponentPose=@n24, Alpha=@n23, Translation="6,-5,-2", Rotation="0,0,0", Scale="1,1,1") {
    translationMode=BMM_Additive
    translationSpace=BCS_BoneSpace
    alphaBoolBlend {blendOption=Linear}
    boneToModify {boneName=dyn_hair_bk_02}
    translation=<6 -5 -2>
   }
   n26 = "Transform (Modify) Bone - Bone: dyn_hair_r_02"(ComponentPose=@n25, Alpha=@n23, Translation="8.661,-7.817,1.643", Rotation="0,0,0", Scale="1,1,1") {
    translationMode=BMM_Additive
    translationSpace=BCS_BoneSpace
    alphaBoolBlend {blendOption=Linear}
    boneToModify {boneName=dyn_hair_r_02}
    translation=<8.661 -7.817 1.643>
   }
   n27 = Anim Dynamics - Bone: dyn_hair_bk_02!(ComponentPose=@n26, Alpha=1, GravityScale:float=1, ExternalForce="0,0,0") -> (Pose) {
    angularDampingOverride=0.9999
    angularSpringConstant=800
    bAngularSpring=true
    bGravityOverrideInSimSpace=true
    bOverrideAngularDamping=true
    bUseGravityOverride=true
    simulationSpace=RootRelative
    alphaBoolBlend {blendOption=Linear}
    boundBone {boneName=dyn_hair_bk_02}
    componentAppliedLinearAccClamp=<1000 1000 1000>
    componentLinearAccScale=<0.2 0.2 0.2>
    componentLinearVelScale=<0.6 0.6 0.6>
    gravityOverride {z=-600}
    physicsBodyDefinitions [{
      sphereCollisionRadius=0
      boundBone {boneName=dyn_hair_bk_02}
      boxExtents=<15 25 20>
      constraintSetup {angularLimitsMax=<0 0 25>, angularLimitsMin=<0 -8 -25>, angularTarget=<0 0 0>}
      localJointOffset=<4 0 0>
     }]
    relativeSpaceBone {boneName=head}
   }
   n28 = Anim Dynamics - Bone: dyn_hair_r_02!(ComponentPose=@n27, Alpha=1, GravityScale:float=1, ExternalForce="0,0,0") -> (Pose) {
    angularDampingOverride=0.9999
    angularSpringConstant=800
    bAngularSpring=true
    bGravityOverrideInSimSpace=true
    bOverrideAngularDamping=true
    bUseGravityOverride=true
    simulationSpace=RootRelative
    alphaBoolBlend {blendOption=Linear}
    boundBone {boneName=dyn_hair_r_02}
    componentAppliedLinearAccClamp=<1000 1000 1000>
    componentLinearAccScale=<0.2 0.2 0.2>
    componentLinearVelScale=<0.6 0.6 0.6>
    gravityOverride {z=-600}
    physicsBodyDefinitions [{
      sphereCollisionRadius=0
      boundBone {boneName=dyn_hair_r_02}
      boxExtents=<15 25 20>
      constraintSetup {angularLimitsMax=<0 0 25>, angularLimitsMin=<0 -8 -10>, angularTarget=<0 0 0>}
      localJointOffset=<4 0 0>
     }]
    relativeSpaceBone {boneName=head}
   }
   n29 = Anim Dynamics - Bone: dyn_hair_bk_l_02!(ComponentPose=@n28, Alpha=1, GravityScale:float=1, ExternalForce="0,0,0") -> (Pose) {
    angularDampingOverride=0.9999
    angularSpringConstant=800
    bAngularSpring=true
    bGravityOverrideInSimSpace=true
    bOverrideAngularDamping=true
    bUseGravityOverride=true
    simulationSpace=RootRelative
    alphaBoolBlend {blendOption=Linear}
    boundBone {boneName=dyn_hair_bk_l_02}
    componentAppliedLinearAccClamp=<1000 1000 1000>
    componentLinearAccScale=<0.2 0.2 0.2>
    componentLinearVelScale=<0.6 0.6 0.6>
    gravityOverride {z=-600}
    physicsBodyDefinitions [{
      sphereCollisionRadius=0
      boundBone {boneName=dyn_hair_bk_l_02}
      boxExtents=<15 25 20>
      constraintSetup {angularLimitsMax=<0 0 20>, angularLimitsMin=<0 -8 -20>, angularTarget=<0 0 0>}
      localJointOffset=<3 0 0>
     }]
    relativeSpaceBone {boneName=head}
   }
   n30 = RigidBody(ComponentPose=@n29, Alpha=0, OverrideWorldGravity="0,0,-980", ComponentLinearAccScale="0.35,0.4,0.4") {
    bClampLinearTranslationLimitToRefPose=true
    bOverrideWorldGravity=true
    overridePhysicsAsset=&/M_MED_Apprentice_Vibrant_FaceAcc_Dyn_Hair_Equipped_Backpack
    simulationSpace=BaseBoneSpace
    alphaBoolBlend {blendOption=Linear}
    alphaScaleBiasClamp {interpSpeedDecreasing=4, interpSpeedIncreasing=6}
    baseBoneRef {boneName=spine_05}
    componentAppliedLinearAccClamp=<3000 3000 3000>
    componentLinearAccScale=<0.6 0.6 0.2>
    componentLinearVelScale=<0.35 0.35 0.35>
    overrideWorldGravity {z=-875}
    simSpaceSettings {maxAngularAcceleration=200, maxAngularVelocity=5, maxLinearAcceleration=500, maxLinearVelocity=200}
   }
   n31 = Component To Local(ComponentPose=@n30)
   n32 = Get head_rot_bk()
   n33 = "Transform (Modify) Bone - Bone: dyn_hair_bk_MSTR"(
    ComponentPose=Local To Component(LocalPose=Use cached pose 'prePoseBaseHair'())
    Alpha=@n32
    Translation="-3,0,1"
    Rotation="0,0,0"
    Scale="1,1,1"
   ) {
    translationMode=BMM_Additive
    translationSpace=BCS_BoneSpace
    alphaBoolBlend {blendOption=Linear}
    boneToModify {boneName=dyn_hair_bk_MSTR}
    translation {x=-3, z=1}
   }
   n34 = "Transform (Modify) Bone - Bone: dyn_hair_r_MSTR"(ComponentPose=@n33, Alpha=@n32, Translation="-3,0,1", Rotation="0,0,0", Scale="1,1,1") {
    translationMode=BMM_Additive
    translationSpace=BCS_BoneSpace
    alphaBoolBlend {blendOption=Linear}
    boneToModify {boneName=dyn_hair_r_MSTR}
    translation {x=-3, z=1}
   }
   n35 = "Transform (Modify) Bone - Bone: dyn_hair_bk_l_MSTR"(ComponentPose=@n34, Alpha=@n32, Translation="-3,0,1", Rotation="0,0,0", Scale="1,1,1") {
    translationMode=BMM_Additive
    translationSpace=BCS_BoneSpace
    alphaBoolBlend {blendOption=Linear}
    boneToModify {boneName=dyn_hair_bk_l_MSTR}
    translation {x=-3, z=1}
   }
   n36 = RigidBody(ComponentPose=@n35, OverrideWorldGravity="0,0,-980", ComponentLinearAccScale="0.35,0.4,0.4") {
    bClampLinearTranslationLimitToRefPose=true
    bOverrideWorldGravity=true
    overridePhysicsAsset=&/M_MED_Apprentice_Vibrant_FaceAcc_Dyn_Hair
    simulationSpace=BaseBoneSpace
    alphaBoolBlend {blendOption=Linear}
    alphaScaleBiasClamp {interpSpeedDecreasing=4, interpSpeedIncreasing=6}
    baseBoneRef {boneName=spine_05}
    componentAppliedLinearAccClamp=<3000 3000 3000>
    componentLinearAccScale=<0.6 0.6 0.2>
    componentLinearVelScale=<0.35 0.35 0.35>
    overrideWorldGravity {z=-875}
    simSpaceSettings {maxAngularAcceleration=200, maxAngularVelocity=5, maxLinearAcceleration=500, maxLinearVelocity=200}
   }
   n37 = Component To Local(ComponentPose=@n36)
   n38 = Get isWearBackblingOrNoRigid() -> (isWearBackblingOrNoRigid:bool)
   Output Pose(Result=Blend Poses by bool(
     BlendPose_0=Apply Additive(Base=@n10, Additive=@n11) {alphaBoolBlend {blendOption=Linear}}
     BlendPose_1=Blend Poses by bool(BlendPose_0=@n31, BlendPose_1=@n37, bActiveValue=@n38) {blendType=Linear}
     BlendTime_0=0.25
     bActiveValue=Get onKill()
    ) {blendType=Linear}) {layerGroup=None}
   n39 = Get SkelMesh()
   Save cached pose 'basePose'(Pose=Layered blend per bone!(BasePose=Copy Pose From Mesh() {bUseAttachedParent=true}, BlendPoses_0=Copy Pose From Mesh(SourceMeshComponent=@n39), BlendWeights_0) -> (Pose) {layerSetup [{branchFilters [{boneName=upperarm_l}, {boneName=upperarm_r}]}]}) {cacheName=basePose}
   Save cached pose 'MeshPose'(Pose=Copy Pose From Mesh(SourceMeshComponent=Get SkelMesh())) {cacheName=MeshPose}
   n40 = Look At - Bone: dyn_hair_bk_MSTR(ComponentPose=Local To Component(LocalPose=Use cached pose 'basePose'()), Alpha=1) {
    alphaBoolBlend {blendOption=Linear}
    boneToModify {boneName=dyn_hair_bk_MSTR}
    lookAtLocation=<-15 -30 3>
    lookAtTarget {boneReference {boneName=spine_05}}
    lookAt_Axis {axis=<1 0>}
   }
   n41 = Look At - Bone: dyn_hair_r_MSTR(ComponentPose=@n40, Alpha=1) {
    alphaBoolBlend {blendOption=Linear}
    boneToModify {boneName=dyn_hair_r_MSTR}
    lookAtLocation=<-15 -20 19>
    lookAtTarget {boneReference {boneName=spine_05}}
    lookAt_Axis {axis=<1 0>}
   }
   n42 = Look At - Bone: dyn_hair_bk_l_MSTR(ComponentPose=@n41, Alpha=1) {
    alphaBoolBlend {blendOption=Linear}
    boneToModify {boneName=dyn_hair_bk_l_MSTR}
    lookAtLocation=<-15 -25 -15>
    lookAtTarget {boneReference {boneName=spine_05}}
    lookAt_Axis {axis=<1 0>}
   }
   n43 = Look At - Bone: dyn_hair_l_MSTR(ComponentPose=@n42, Alpha=1) {
    alphaBoolBlend {blendOption=Linear}
    boneToModify {boneName=dyn_hair_l_MSTR}
    lookAtLocation=<-15 4.2 -16.5>
    lookAtTarget {boneReference {boneName=spine_05}}
    lookAt_Axis {axis=<1 0>}
   }
   Save cached pose 'prePoseBaseHair'(Pose=Component To Local(ComponentPose=@n43)) {cacheName=prePoseBaseHair}
  ]}
 EventGraph {nodes [
   Event BlueprintUpdateAnimation()
   > n1 = Sequence!(execute) -> (then_0, then_1, then_2:Exec)
   > n2 = Get Dynamic Material Instance(SkeletalMeshComponent=GetPartSkeletalMeshComponent(), ElementIndex=1)
   > n3 = Get Scalar Parameter Value!(execute, self:Material Instance Dynamic Object Reference=@n2, ParameterName:Name=Hide Element 02) -> (then, ReturnValue:float) {is=title K2Node_CallFunction}
   > Set ReactiveAlpha(ReactiveAlpha:double=MapRangeClamped(Value=@n3, InRangeA=0, InRangeB=1, OutRangeA=1, OutRangeB=0)) -> (Output_Get:double)
   n4 = Get SkelMesh()
   Set head_rot_bk(
    execute=@n1.then_1
    head_rot_bk:double=GetAlphaFromReferenceTransformDelta(SkeletalMeshComponent=@n4, BoneName=head, BaseName=spine_05, DeltaComponentType=QuaternionTwist, InputRange=(X=0,Y=60), TwistAxis="0,0,1")
   ) -> (Output_Get:double)
   > Set head_twist_l(head_twist_l:double=GetAlphaFromReferenceTransformDelta(SkeletalMeshComponent=@n4, BoneName=head, BaseName=spine_05, DeltaComponentType=QuaternionTwist, InputRange=(X=0,Y=50), TwistAxis="1,0,0")) -> (Output_Get:double)
   > Set head_rot_r(head_rot_r:double=GetAlphaFromReferenceTransformDelta(SkeletalMeshComponent=@n4, BoneName=head, BaseName=spine_05, DeltaComponentType=QuaternionTwist, InputRange=(X=0,Y=40), TwistAxis="0,1,0")) -> (Output_Get:double)
   n5 = Get bRigidBodyEnabled()
   Set isWearBackblingOrNoRigid(execute=@n1.then_2, isWearBackblingOrNoRigid:bool=OR Boolean(A=Get isWearBackbling() -> (isWearBackbling:bool), B=NOT Boolean(A=@n5))) -> (Output_Get:bool)
   Event OnCharacterCustomizationLoaded()
   > Do Once()
   > Set isWearBackbling(isWearBackbling:bool=CheckForBackbling()) -> (Output_Get:bool)
   n6 = Event BlueprintProcessAnimInputEvent()
   > Branch(Condition=IsNamed(self=@n6.AnimInputEvent, InEventName=aie.EnemyPlayerEliminated))
   > Set onKill(onKill:bool=true) -> (Output_Get:bool)
   > RetriggerableDelay(Duration=6)
   > Set onKill(onKill=false)
   Event BlueprintInitializeAnimation()
   n7 = GetOwnerSkeletalMeshComponent()
   > Is Valid(InputObject=@n7)
   > Set SkelMesh(SkelMesh=@n7)
  ]}
]
variables [
 head_rot_bk {type=number}
 head_rot_r {type=number}
 head_twist_l {type=number}
 isWearBackbling {type=boolean}
 isWearBackblingOrNoRigid {type=boolean}
 onKill {type=boolean}
 ReactiveAlpha {type=number}
 SkelMesh {type=SkeletalMeshComponent}
]
